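# Physics asset for the face accessory mesh of the male medium Fantastical Purple Headset player character, covering 17 dynamic hair pieces.
# Each piece is a two-body chain: a small master capsule constrained to the head with free swing and twist, and a sized capsule constrained to that master.
# The second link has per-piece swing limits and a zero twist limit.
# One large, nearly spherical capsule represents the head.
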PhysicsAsset /Game/Characters/Player/Male/Medium/Bodies/M_MED_Headset_Fantastical_Purple/Meshes/Parts/M_MED_Headset_Fantastical_Purple_FaceAcc_Physics
bodies {
 dyn_hairMesh_1 [{shapeType=Capsule, length=1.7984, radius=1.223, shapeName=dyn_hairMesh_1_capsule, center=<2 0 0>, rotation=<rot 90 0 0>}]
 dyn_hairMesh_10 [{shapeType=Capsule, length=1.5882, radius=0.5699, shapeName=dyn_hairMesh_10_capsule, center=<2 0 0>, rotation=<rot 90 0 0>}]
 dyn_hairMesh_10_MSTR [{shapeType=Capsule, length=0.5, radius=0.5, shapeName=dyn_hairMesh_10_MSTR_capsule, rotation=<rot 0 -90 0>}]
 dyn_hairMesh_11 [{shapeType=Capsule, length=2.049, radius=1.1708, shapeName=dyn_hairMesh_11_capsule, center=<2 0 0>, rotation=<rot 90 0 0>}]
 dyn_hairMesh_11_MSTR [{shapeType=Capsule, length=0.5, radius=0.5, shapeName=dyn_hairMesh_11_MSTR_capsule, rotation=<rot 0 -90 0>}]
 dyn_hairMesh_12 [{shapeType=Capsule, length=2.1327, radius=1.1159, shapeName=dyn_hairMesh_12_capsule, center=<2 0 0>, rotation=<rot 90 0 0>}]
 dyn_hairMesh_12_MSTR [{shapeType=Capsule, length=0.5, radius=0.5, shapeName=dyn_hairMesh_12_MSTR_capsule, rotation=<rot 0 -90 0>}]
 dyn_hairMesh_13 [{shapeType=Capsule, length=1.21, radius=1.1044, shapeName=dyn_hairMesh_13_capsule, center=<2 0 0>, rotation=<rot 90 0 0>}]
 dyn_hairMesh_13_MSTR [{shapeType=Capsule, length=0.5, radius=0.5, shapeName=dyn_hairMesh_13_MSTR_capsule, rotation=<rot 0 -90 0>}]
 dyn_hairMesh_14 [{shapeType=Capsule, length=1.8371, radius=0.6054, shapeName=dyn_hairMesh_14_capsule, center=<2 0 0>, rotation=<rot 90 0 0>}]
 dyn_hairMesh_14_MSTR [{shapeType=Capsule, length=0.5, radius=0.5, shapeName=dyn_hairMesh_14_MSTR_capsule, rotation=<rot 0 -90 0>}]
 dyn_hairMesh_15 [{shapeType=Capsule, length=0.505, radius=0.505, shapeName=dyn_hairMesh_15_capsule, center=<2 0 0>, rotation=<rot 90 0 0>}]
 dyn_hairMesh_15_MSTR [{shapeType=Capsule, length=0.5, radius=0.5, shapeName=dyn_hairMesh_15_MSTR_capsule, rotation=<rot 0 -90 0>}]
 dyn_hairMesh_16 [{shapeType=Capsule, length=1.3762, radius=0.7079, shapeName=dyn_hairMesh_16_capsule, center=<2 0 0>, rotation=<rot 90 0 0>}]
 dyn_hairMesh_16_MSTR [{shapeType=Capsule, length=0.5, radius=0.5, shapeName=dyn_hairMesh_16_MSTR_capsule, rotation=<rot 0 -90 0>}]
 dyn_hairMesh_17 [{shapeType=Capsule, length=3.3586, radius=0.9465, shapeName=dyn_hairMesh_17_capsule, center=<2 0 0>, rotation=<rot 90 0 0>}]
 dyn_hairMesh_17_MSTR [{shapeType=Capsule, length=0.5, radius=0.5, shapeName=dyn_hairMesh_17_MSTR_capsule, rotation=<rot 0 -90 0>}]
 dyn_hairMesh_1_MSTR [{shapeType=Capsule, length=0.5, radius=0.5, shapeName=dyn_hairMesh_1_MSTR_capsule, rotation=<rot 0 -90 0>}]
 dyn_hairMesh_2 [{shapeType=Capsule, length=0.9262, radius=0.6641, shapeName=dyn_hairMesh_2_capsule, center=<2 0 0>, rotation=<rot 90 0 0>}]
 dyn_hairMesh_2_MSTR [{shapeType=Capsule, length=0.5, radius=0.5, shapeName=dyn_hairMesh_2_MSTR_capsule, rotation=<rot 0 -90 0>}]
 dyn_hairMesh_3 [{shapeType=Capsule, length=1.0757, radius=0.5022, shapeName=dyn_hairMesh_3_capsule, center=<2 0 0>, rotation=<rot 90 0 0>}]
 dyn_hairMesh_3_MSTR [{shapeType=Capsule, length=0.5, radius=0.5, shapeName=dyn_hairMesh_3_MSTR_capsule, rotation=<rot 0 -90 0>}]
 dyn_hairMesh_4 [{shapeType=Capsule, length=1.7247, radius=1.3004, shapeName=dyn_hairMesh_4_capsule, center=<2 0 0>, rotation=<rot 90 0 0>}]
 dyn_hairMesh_4_MSTR [{shapeType=Capsule, length=0.5, radius=0.5, shapeName=dyn_hairMesh_4_MSTR_capsule, rotation=<rot 0 -90 0>}]
 dyn_hairMesh_5 [{shapeType=Capsule, length=1.0896, radius=0.5551, shapeName=dyn_hairMesh_5_capsule, center=<2 0 0>, rotation=<rot 90 0 0>}]
 dyn_hairMesh_5_MSTR [{shapeType=Capsule, length=0.5, radius=0.5, shapeName=dyn_hairMesh_5_MSTR_capsule, rotation=<rot 0 -90 0>}]
 dyn_hairMesh_6 [{shapeType=Capsule, length=0.505, radius=0.505, shapeName=dyn_hairMesh_6_capsule, center=<2 0 0>, rotation=<rot 90 0 0>}]
 dyn_hairMesh_6_MSTR [{shapeType=Capsule, length=0.5, radius=0.5, shapeName=dyn_hairMesh_6_MSTR_capsule, rotation=<rot 0 -90 0>}]
 dyn_hairMesh_7 [{shapeType=Capsule, length=2.624, radius=0.5, shapeName=dyn_hairMesh_7_capsule, center=<2 0 0>, rotation=<rot 90 0 0>}]
 dyn_hairMesh_7_MSTR [{shapeType=Capsule, length=0.5, radius=0.5, shapeName=dyn_hairMesh_7_MSTR_capsule, rotation=<rot 0 -90 0>}]
 dyn_hairMesh_8 [{shapeType=Capsule, length=1.4385, radius=0.6131, shapeName=dyn_hairMesh_8_capsule, center=<2 0 0>, rotation=<rot 90 0 0>}]
 dyn_hairMesh_8_MSTR [{shapeType=Capsule, length=0.5, radius=0.5, shapeName=dyn_hairMesh_8_MSTR_capsule, rotation=<rot 0 -90 0>}]
 dyn_hairMesh_9 [{shapeType=Capsule, length=1.5128, radius=0.5684, shapeName=dyn_hairMesh_9_capsule, center=<2 0 0>, rotation=<rot 90 0 0>}]
 dyn_hairMesh_9_MSTR [{shapeType=Capsule, length=0.5, radius=0.5, shapeName=dyn_hairMesh_9_MSTR_capsule, rotation=<rot 0 -90 0>}]
 head [{shapeType=Capsule, length=0.124, radius=10.3429, center=<0.8171 1.0987 0>}]
}
constraints [
 {
  bone1Name=dyn_hairMesh_10_MSTR
  bone2Name=head
  swing1LimitDegrees=45
  swing1Motion=Free
  swing2LimitDegrees=45
  swing2Motion=Free
  twistLimitDegrees=45
  twistMotion=Free
 }
 {bone1Name=dyn_hairMesh_10, bone2Name=dyn_hairMesh_10_MSTR, swing1LimitDegrees=20, swing2LimitDegrees=20, twistLimitDegrees=0}
 {
  bone1Name=dyn_hairMesh_11_MSTR
  bone2Name=head
  swing1LimitDegrees=45
  swing1Motion=Free
  swing2LimitDegrees=45
  swing2Motion=Free
  twistLimitDegrees=45
  twistMotion=Free
 }
 {bone1Name=dyn_hairMesh_11, bone2Name=dyn_hairMesh_11_MSTR, swing1LimitDegrees=20, swing2LimitDegrees=20, twistLimitDegrees=0}
 {
  bone1Name=dyn_hairMesh_12_MSTR
  bone2Name=head
  swing1LimitDegrees=45
  swing1Motion=Free
  swing2LimitDegrees=45
  swing2Motion=Free
  twistLimitDegrees=45
  twistMotion=Free
 }
 {bone1Name=dyn_hairMesh_12, bone2Name=dyn_hairMesh_12_MSTR, swing1LimitDegrees=8, swing2LimitDegrees=8, twistLimitDegrees=0}
 {
  bone1Name=dyn_hairMesh_13_MSTR
  bone2Name=head
  swing1LimitDegrees=45
  swing1Motion=Free
  swing2LimitDegrees=45
  swing2Motion=Free
  twistLimitDegrees=45
  twistMotion=Free
 }
 {bone1Name=dyn_hairMesh_13, bone2Name=dyn_hairMesh_13_MSTR, swing1LimitDegrees=10, swing2LimitDegrees=10, twistLimitDegrees=0}
 {
  bone1Name=dyn_hairMesh_14_MSTR
  bone2Name=head
  swing1LimitDegrees=45
  swing1Motion=Free
  swing2LimitDegrees=45
  swing2Motion=Free
  twistLimitDegrees=45
  twistMotion=Free
 }
 {bone1Name=dyn_hairMesh_14, bone2Name=dyn_hairMesh_14_MSTR, swing1LimitDegrees=20, swing2LimitDegrees=20, twistLimitDegrees=0}
 {
  bone1Name=dyn_hairMesh_15_MSTR
  bone2Name=head
  swing1LimitDegrees=45
  swing1Motion=Free
  swing2LimitDegrees=45
  swing2Motion=Free
  twistLimitDegrees=45
  twistMotion=Free
 }
 {bone1Name=dyn_hairMesh_15, bone2Name=dyn_hairMesh_15_MSTR, swing1LimitDegrees=20, swing2LimitDegrees=20, twistLimitDegrees=0}
 {
  bone1Name=dyn_hairMesh_16_MSTR
  bone2Name=head
  swing1LimitDegrees=45
  swing1Motion=Free
  swing2LimitDegrees=45
  swing2Motion=Free
  twistLimitDegrees=45
  twistMotion=Free
 }
 {bone1Name=dyn_hairMesh_16, bone2Name=dyn_hairMesh_16_MSTR, swing1LimitDegrees=20, swing2LimitDegrees=20, twistLimitDegrees=0}
 {
  bone1Name=dyn_hairMesh_17_MSTR
  bone2Name=head
  swing1LimitDegrees=45
  swing1Motion=Free
  swing2LimitDegrees=45
  swing2Motion=Free
  twistLimitDegrees=45
  twistMotion=Free
 }
 {bone1Name=dyn_hairMesh_17, bone2Name=dyn_hairMesh_17_MSTR, swing1LimitDegrees=15, swing2LimitDegrees=15, twistLimitDegrees=0}
 {
  bone1Name=dyn_hairMesh_1_MSTR
  bone2Name=head
  swing1LimitDegrees=45
  swing1Motion=Free
  swing2LimitDegrees=45
  swing2Motion=Free
  twistLimitDegrees=45
  twistMotion=Free
 }
 {bone1Name=dyn_hairMesh_1, bone2Name=dyn_hairMesh_1_MSTR, swing1LimitDegrees=20, swing2LimitDegrees=20, twistLimitDegrees=0}
 {
  bone1Name=dyn_hairMesh_2_MSTR
  bone2Name=head
  swing1LimitDegrees=45
  swing1Motion=Free
  swing2LimitDegrees=45
  swing2Motion=Free
  twistLimitDegrees=45
  twistMotion=Free
 }
 {bone1Name=dyn_hairMesh_2, bone2Name=dyn_hairMesh_2_MSTR, swing1LimitDegrees=10, swing2LimitDegrees=10, twistLimitDegrees=0}
 {
  bone1Name=dyn_hairMesh_3_MSTR
  bone2Name=head
  swing1LimitDegrees=45
  swing1Motion=Free
  swing2LimitDegrees=45
  swing2Motion=Free
  twistLimitDegrees=45
  twistMotion=Free
 }
 {bone1Name=dyn_hairMesh_3, bone2Name=dyn_hairMesh_3_MSTR, swing1LimitDegrees=20, swing2LimitDegrees=20, twistLimitDegrees=0}
 {
  bone1Name=dyn_hairMesh_4_MSTR
  bone2Name=head
  swing1LimitDegrees=45
  swing1Motion=Free
  swing2LimitDegrees=45
  swing2Motion=Free
  twistLimitDegrees=45
  twistMotion=Free
 }
 {bone1Name=dyn_hairMesh_4, bone2Name=dyn_hairMesh_4_MSTR, swing1LimitDegrees=10, swing2LimitDegrees=10, twistLimitDegrees=0}
 {
  bone1Name=dyn_hairMesh_5_MSTR
  bone2Name=head
  swing1LimitDegrees=45
  swing1Motion=Free
  swing2LimitDegrees=45
  swing2Motion=Free
  twistLimitDegrees=45
  twistMotion=Free
 }
 {bone1Name=dyn_hairMesh_5, bone2Name=dyn_hairMesh_5_MSTR, swing1LimitDegrees=20, swing2LimitDegrees=20, twistLimitDegrees=0}
 {
  bone1Name=dyn_hairMesh_6_MSTR
  bone2Name=head
  swing1LimitDegrees=45
  swing1Motion=Free
  swing2LimitDegrees=45
  swing2Motion=Free
  twistLimitDegrees=45
  twistMotion=Free
 }
 {bone1Name=dyn_hairMesh_6, bone2Name=dyn_hairMesh_6_MSTR, swing1LimitDegrees=20, swing2LimitDegrees=20, twistLimitDegrees=0}
 {
  bone1Name=dyn_hairMesh_7_MSTR
  bone2Name=head
  swing1LimitDegrees=45
  swing1Motion=Free
  swing2LimitDegrees=45
  swing2Motion=Free
  twistLimitDegrees=45
  twistMotion=Free
 }
 {bone1Name=dyn_hairMesh_7, bone2Name=dyn_hairMesh_7_MSTR, swing1LimitDegrees=30, swing2LimitDegrees=30, twistLimitDegrees=0}
 {
  bone1Name=dyn_hairMesh_8_MSTR
  bone2Name=head
  swing1LimitDegrees=45
  swing1Motion=Free
  swing2LimitDegrees=45
  swing2Motion=Free
  twistLimitDegrees=45
  twistMotion=Free
 }
 {bone1Name=dyn_hairMesh_8, bone2Name=dyn_hairMesh_8_MSTR, swing1LimitDegrees=20, swing2LimitDegrees=20, twistLimitDegrees=0}
 {
  bone1Name=dyn_hairMesh_9_MSTR
  bone2Name=head
  swing1LimitDegrees=45
  swing1Motion=Free
  swing2LimitDegrees=45
  swing2Motion=Free
  twistLimitDegrees=45
  twistMotion=Free
 }
 {bone1Name=dyn_hairMesh_9, bone2Name=dyn_hairMesh_9_MSTR, swing1LimitDegrees=20, swing2LimitDegrees=20, twistLimitDegrees=0}
]
